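# Material function that adds an animated emissive visor to the SynthContact character's material attributes, behind a static switch.
# A distance-field flipbook, framed by time or event voice audio amplitude, is thresholded into sharp shapes and blended with curve-atlas-animated dots from a cropped mask.
# A second, bump-offset copy adds parallax depth.
# Both layers are tinted by an emissive texture, combined by maximum, channel-masked, then added to the incoming emissive.
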MaterialFunction /BRCosmetics/Characters/Player/Male/Medium/Bodies/M_MED_SynthContact/Materials/MF_SynthContactVisorStyles1
graph {nodes [
  n1 = GetMaterialAttributes(None=FunctionInput() {inputName=Material, inputType=FunctionInput_MaterialAttributes}) {attributeGetTypes [#1]}
  n2 = FunctionInput(Preview=Constant() {r=1}) {bUsePreviewValueAsDefault=true, inputName=Mask, inputType=FunctionInput_Vector4}
  n3 = TextureObjectParameter() {
   group=Visor
   parameterName=Visor_EmissiveDistanceField
   samplerType=SAMPLERTYPE_Alpha
   texture=/BRCosmetics/Characters/Player/Male/Medium/Bodies/M_MED_SynthContact/Textures/T_SynthContact_Visor_02_Flipbook
  }
  n4 = TextureCoordinate()
  n5 = ScalarParameter() {defaultValue=2, group=Visor, parameterName=Visor_SubUV_Frames}
  n6 = Multiply(
   A=StaticSwitchParameter(
    True=CollectionParameter() {collection=/Game/Athena/Events/AudioAnalysis/MPC_Audio_Reactive_Parameters, parameterName=Amplitude_VO_B}
    False=Time()
   ) {group=SmartTuna, parameterName=UseEventVO}
   B=ScalarParameter() {defaultValue=2, group=Visor, parameterName=Visor_SubUV_Speed}
  )
  n7 = ScalarParameter() {defaultValue=-0.48, group=Visor, parameterName=Visor_LineThick}
  n8 = Add(
   A=MaterialFunctionCall(Float3=MaterialFunctionCall!(Texture=@n3, UVs=@n4, SubImages=@n5, Frame=@n6) {materialFunction=/Engine/Functions/Engine_MaterialFunctions03/Texturing/SubUV_Function}.RGB) {materialFunction=&BreakOutFloat3Components}.R
   B=@n7
  ) {constB=-0.48}
  n9 = Saturate(None=Multiply(A=Saturate(None=@n8), B=Constant() {r=16}))
  n10 = NamedRerouteUsage()
  n11 = NamedRerouteUsage()
  n12 = TextureObjectParameter() {group=Visor, parameterName=Visor_Emissive, texture=&T_BlacksRGB}
  n13 = TextureSample(UVs=@n4, Tex=@n12)
  n14 = Multiply(
   A=Multiply(
    A=LinearInterpolate(A=@n9, B=@n10, Alpha=@n11)
    B=StaticComponentMaskParameter(None=@n13.RGB) {defaultB=true, defaultG=true, defaultR=true, group=Visor, parameterName=Visor_Emissive Component Mask}
   )
   B=ScalarParameter() {defaultValue=5, group=Visor, parameterName=Visor_Layer1_Strength}
  )
  n15 = BumpOffset(
   Coordinate=TextureCoordinate()
   Height=ScalarParameter() {group=Visor, parameterName=VisorHeight}
   HeightRatioInput=ScalarParameter() {defaultValue=0.015, parameterName=VisorLayer2HeightRatio}
  ) {referencePlane=1}
  n16 = Add(
   A=MaterialFunctionCall(Float3=MaterialFunctionCall(Texture=@n3, UVs=@n15, SubImages=@n5, Frame=@n6) {materialFunction=&SubUV_Function}.RGB) {materialFunction=&BreakOutFloat3Components}.R
   B=@n7
  ) {constB=-0.48}
  n17 = Saturate(None=Multiply(A=Saturate(None=@n16), B=Constant() {r=16}))
  n18 = NamedRerouteUsage()
  n19 = NamedRerouteUsage()
  n20 = TextureSample(UVs=@n15, Tex=@n12)
  n21 = Multiply(
   A=Multiply(
    A=LinearInterpolate(A=@n17, B=@n18, Alpha=@n19)
    B=StaticComponentMaskParameter(None=@n20.RGB) {defaultB=true, defaultG=true, defaultR=true, group=Visor, parameterName=Visor_Emissive Component Mask}
   )
   B=ScalarParameter() {defaultValue=1, group=Visor, parameterName=Visor_Layer2_Strength}
  )
  n22 = Add(
   A=@n1.EmissiveColor
   B=Multiply(
    A=ChannelMaskParameter(None=@n2) {group=Visor, maskChannel=Green, parameterName=VisorMaskChannel, defaultValue=<rgba 0 1 0 0>}.R
    B=Max(A=@n14, B=@n21)
   )
  )
  FunctionOutput(None=SetMaterialAttributes!(
    MaterialAttributes=@n1.MaterialAttributes
    EmissiveColor=StaticSwitchParameter(True=@n22, False=@n1.EmissiveColor) {parameterName=UseAnimatedVisor}
   ) {attributeSetTypes [#1]})
  n23 = TextureCoordinate()
  n24 = TextureObject() {
   samplerType=SAMPLERTYPE_Masks
   texture=/BRCosmetics/Characters/Player/Male/Medium/Bodies/M_MED_SynthContact/Textures/T_SynthContact_Visor_02_Dots
  }
  n25 = Constant2Vector() {g=0.62, r=0.39}
  n26 = Constant2Vector() {g=0.85, r=0.61}
  n27 = MaterialFunctionCall(Float2=Multiply(A=MaterialFunctionCall!(UVs=@n23, TextureIn=@n24, UpperLeftCorner=@n25, LowerRightCorner=@n26, ExportFloat4) {materialFunction=/Engine/Functions/Engine_MaterialFunctions01/Texturing/TextureCropping}.CroppedMasked)) {materialFunction=&BreakOutFloat2Components}
  NamedRerouteDeclaration(None=Sign(None=Multiply(A=@n27.G))) {name=DotArea, nodeColor=<rgba 1 0.5156 0 1>}
  n28 = Time()
  n29 = ScalarParameter() {defaultValue=0.4, parameterName=DotAnimSpeed}
  n30 = CurveAtlasRowParameter(CurveTime=Frac(None=Multiply(A=@n28, B=@n29))) {
   atlas=&/CA_SynthContact
   curve=&/CLC_SynthContact
   parameterName=DotAnimationCurve
  }
  n31 = Saturate(None=Step(Y=@n30.R, X=@n27.R) {constX=0.1, constY=0.01})
  n32 = Saturate(None=Add(A=Multiply(A=@n31, B=@n27.G) {constA=1}, B=ScalarParameter() {defaultValue=-0.4, group=Visor, parameterName=Visor_LineThick2}) {constB=-0.48})
  n33 = Constant() {r=16}
  NamedRerouteDeclaration(None=Saturate(None=Multiply(A=@n32, B=@n33))) {name=Dot, nodeColor=<rgba 0.1953 0 1 1>}
  n34 = BumpOffset(
   Coordinate=TextureCoordinate()
   Height=ScalarParameter() {group=Visor, parameterName=VisorHeight}
   HeightRatioInput=ScalarParameter() {defaultValue=0.015, parameterName=VisorLayer2HeightRatio}
  ) {referencePlane=1}
  n35 = TextureObject() {
   samplerType=SAMPLERTYPE_Masks
   texture=&T_SynthContact_Visor_02_Dots
  }
  n36 = Constant2Vector() {g=0.62, r=0.39}
  n37 = Constant2Vector() {g=0.85, r=0.61}
  n38 = MaterialFunctionCall(Float2=Multiply(A=MaterialFunctionCall(UVs=@n34, TextureIn=@n35, UpperLeftCorner=@n36, LowerRightCorner=@n37) {materialFunction=&TextureCropping}.CroppedMasked)) {materialFunction=&BreakOutFloat2Components}
  NamedRerouteDeclaration(None=Sign(None=Multiply(A=@n38.G))) {name=DotArea_1, nodeColor=<rgba 1 0.5156 0 1>}
  n39 = Time()
  n40 = ScalarParameter() {defaultValue=0.4, parameterName=DotAnimSpeed}
  n41 = CurveAtlasRowParameter(CurveTime=Frac(None=Multiply(A=@n39, B=@n40))) {
   atlas=&CA_SynthContact
   curve=&CLC_SynthContact
   parameterName=DotAnimationCurve
  }
  n42 = Saturate(None=Step(Y=@n41.R, X=@n38.R) {constX=0.1, constY=0.01})
  n43 = Saturate(None=Add(A=Multiply(A=@n42, B=@n38.G) {constA=1}, B=ScalarParameter() {defaultValue=-0.4, group=Visor, parameterName=Visor_LineThick2}) {constB=-0.48})
  n44 = Constant() {r=16}
  NamedRerouteDeclaration(None=Saturate(None=Multiply(A=@n43, B=@n44))) {name=Dot_Ly2, nodeColor=<rgba 0.1953 0 1 1>}
  ScalarParameter() {defaultValue=0.02, group=Visor, parameterName=VisorLayerHeightRatio}
  Time()
 ]}
