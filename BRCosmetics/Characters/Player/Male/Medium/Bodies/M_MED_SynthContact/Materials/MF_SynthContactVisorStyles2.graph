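# Material function for the SynthContact character's visor (style 2) that adds an animated flipbook glow to the incoming material's emissive, behind a static switch.
# Time or a voice-over amplitude collection parameter, picked by static switch, drives a curve atlas lookup that selects the SubUV frame.
# The distance-field frame is sharpened, tinted and sampled twice, once with BumpOffset for parallax depth.
# The brighter layer wins via Max and is limited by a mask channel.
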MaterialFunction /BRCosmetics/Characters/Player/Male/Medium/Bodies/M_MED_SynthContact/Materials/MF_SynthContactVisorStyles2
graph {nodes [
  n1 = GetMaterialAttributes(None=FunctionInput() {inputName=Material, inputType=FunctionInput_MaterialAttributes}) {attributeGetTypes [#1]}
  n2 = FunctionInput(Preview=Constant() {r=1}) {bUsePreviewValueAsDefault=true, inputName=Mask, inputType=FunctionInput_Vector4}
  n3 = TextureObjectParameter() {
   group=Visor
   parameterName=Visor_EmissiveDistanceFieldStyle2
   samplerType=SAMPLERTYPE_Alpha
   texture=/BRCosmetics/Characters/Player/Male/Medium/Bodies/M_MED_SynthContact/Textures/T_SynthContact_Visor_03_Flipbook
  }
  n4 = TextureCoordinate()
  n5 = ScalarParameter() {defaultValue=2, group=Visor, parameterName=Visor_SubUV_Frames}
  n6 = StaticSwitchParameter(
   True=CollectionParameter() {collection=/Game/Athena/Events/AudioAnalysis/MPC_Audio_Reactive_Parameters, parameterName=Amplitude_VO_B}
   False=Time()
  ) {group=SmartTuna, parameterName=UseEventVO}
  n7 = ScalarParameter() {defaultValue=0.2, parameterName=DotAnimSpeedStyle2}
  n8 = CurveAtlasRowParameter(CurveTime=Frac(None=Multiply(A=@n6, B=@n7))) {
   atlas=&/CA_SynthContact
   curve=&/CLC_SynthContact
   parameterName=DotAnimationCurve
  }
  n9 = MaterialFunctionCall(Float3=MaterialFunctionCall!(Texture=@n3, UVs=@n4, SubImages=@n5, Frame=@n8.G) {materialFunction=/Engine/Functions/Engine_MaterialFunctions03/Texturing/SubUV_Function}.RGB) {materialFunction=&BreakOutFloat3Components}
  n10 = ScalarParameter() {defaultValue=-0.48, group=Visor, parameterName=Visor_LineThick}
  n11 = Multiply(A=Saturate(None=Add(A=@n9.R, B=@n10) {constB=-0.48}), B=Constant() {r=16})
  n12 = TextureObjectParameter() {group=Visor, parameterName=Visor_Emissive, texture=&T_BlacksRGB}
  n13 = TextureSample(UVs=@n4, Tex=@n12)
  n14 = Multiply(
   A=Multiply(
    A=Saturate(None=@n11)
    B=StaticComponentMaskParameter(None=@n13.RGB) {defaultB=true, defaultG=true, defaultR=true, group=Visor, parameterName=Visor_Emissive Component Mask}
   )
   B=ScalarParameter() {defaultValue=5, group=Visor, parameterName=Visor_Layer1_Strength}
  )
  n15 = BumpOffset(
   Coordinate=TextureCoordinate()
   Height=ScalarParameter() {group=Visor, parameterName=VisorHeight}
   HeightRatioInput=ScalarParameter() {defaultValue=0.015, parameterName=VisorLayer2HeightRatio}
  ) {referencePlane=1}
  n16 = MaterialFunctionCall(Float3=MaterialFunctionCall(Texture=@n3, UVs=@n15, SubImages=@n5, Frame=@n8.G) {materialFunction=&SubUV_Function}.RGB) {materialFunction=&BreakOutFloat3Components}
  n17 = Multiply(A=Saturate(None=Add(A=@n16.R, B=@n10) {constB=-0.48}), B=Constant() {r=16})
  n18 = TextureSample(UVs=@n15, Tex=@n12)
  n19 = Multiply(
   A=Multiply(
    A=Saturate(None=@n17)
    B=StaticComponentMaskParameter(None=@n18.RGB) {defaultB=true, defaultG=true, defaultR=true, group=Visor, parameterName=Visor_Emissive Component Mask}
   )
   B=ScalarParameter() {defaultValue=1, group=Visor, parameterName=Visor_Layer2_Strength}
  )
  n20 = Add(
   A=@n1.EmissiveColor
   B=Multiply(
    A=ChannelMaskParameter(None=@n2) {group=Visor, maskChannel=Green, parameterName=VisorMaskChannel, defaultValue=<rgba 0 1 0 0>}.R
    B=Max(A=@n14, B=@n19)
   )
  )
  FunctionOutput(None=SetMaterialAttributes!(
    MaterialAttributes=@n1.MaterialAttributes
    EmissiveColor=StaticSwitchParameter(True=@n20, False=@n1.EmissiveColor) {parameterName=UseAnimatedVisor}
   ) {attributeSetTypes [#1]})
  Multiply(A=Time(), B=ScalarParameter() {defaultValue=2, group=Visor, parameterName=Visor_SubUV_Speed})
  Multiply(A=@n8.G)
  ScalarParameter() {defaultValue=0.02, group=Visor, parameterName=VisorLayerHeightRatio}
  Time()
 ]}
