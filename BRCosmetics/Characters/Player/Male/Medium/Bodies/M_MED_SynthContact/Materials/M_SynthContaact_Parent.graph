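# Parent material for the SynthContact player character body, providing selectable visor looks on top of the shared 2023 character shading.
# The CharacterParent PreFX function builds material attributes, which feed three visor functions: SynthContact Styles1, Styles2, and MF_Visor borrowed from the female Binary body.
# Nested static switches IsStyle1 and IsStyle2 pick one, defaulting to MF_Visor when both are off.
# The chosen result enters the CharacterParent PostFX function alongside PreFX's auxiliary output.
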Material /BRCosmetics/Characters/Player/Male/Medium/Bodies/M_MED_SynthContact/Materials/M_SynthContaact_Parent
graph {
 properties {
  bUseMaterialAttributes=true
  bUsedWithMeshDeformer=true
  bUsedWithMorphTargets=true
  bUsedWithSkeletalMesh=true
  pixelDepthOffsetMode=PDOM_Legacy
  shadingModel=MSM_FromMaterialExpression
  usedShadingModels=Default Lit | Subsurface | Clear Coat | Subsurface Profile | Cloth
 }
 nodes [
  n1 = MaterialFunctionCall() {materialFunction=&MF_CharacterParent_2023_PreFX}
  n2 = MaterialFunctionCall!(Mask, Material=@n1.Result) {materialFunction=&/MF_SynthContactVisorStyles2}
  n3 = MaterialFunctionCall!(Mask, Material=@n1.Result) {materialFunction=/Game/Characters/Player/Female/Medium/Bodies/F_MED_Binary/Materials/MF_Visor}
  MaterialFunctionCall(
   Material=StaticSwitchParameter(
    True=MaterialFunctionCall!(Mask, Material=@n1.Result) {materialFunction=&/MF_SynthContactVisorStyles1}.Result
    False=StaticSwitchParameter(True=@n2.Result, False=@n3.Result) {parameterName=IsStyle2}
   ) {parameterName=IsStyle1}
   M=@n1.M
   Pre Skinned Local Position=@n1.M
   Distance Blend=@n1.M
   Fresnel=@n1.M
  ) {materialFunction=&MF_CharacterParent_2023_PostFX}
 ]
}
