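# Physics asset for the medium male SynthContact body cosmetic, setting up dangling motion for a tube and a pair of pouches.
# The tube is two capsule chains, back from spine_03 and front from spine_05, closed into a loop at dyn_tube_c by constraints in both directions.
# Front and back pouches share a pivot on spine_01.
# All joints swing freely, and capsules on the spine and left arm provide body shapes.
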PhysicsAsset /BRCosmetics/Characters/Player/Male/Medium/Bodies/M_MED_SynthContact/Meshes/M_MED_SynthContact_Physics
bodies {
 dyn_pouch_bck [{
   shapeType=Capsule
   length=6.6691
   radius=3.5686
   shapeName=dyn_pouch_bck_capsule
   center=<2.6655 -0.0786 -2.0591>
   rotation=<rot 3.5065 -90 -11.3167>
  }]
 dyn_pouch_fr [{
   shapeType=Capsule
   length=6.6691
   radius=3.5378
   shapeName=dyn_pouch_fr_capsule
   center=<2.5886 -0.0627 -1.9835>
   rotation=<rot 0.2289 -90 -13.9457>
  }]
 dyn_pouch_rotate [{shapeType=Capsule, length=0.505, radius=0.505, shapeName=dyn_pouch_rotate_capsule, rotation=<rot 0 -90 0>}]
 dyn_tube_bck_01 [{shapeType=Capsule, length=0.505, radius=0.505, shapeName=dyn_tube_bck_01_capsule, rotation=<rot 0 -90 0>}]
 dyn_tube_bck_02 [{shapeType=Capsule, length=3, radius=2, shapeName=dyn_tube_bck_02_capsule, center=<-0.7929 0 0>, rotation=<rot 90 0 0>}]
 dyn_tube_bck_03 [{shapeType=Capsule, length=3.5, radius=1.5, shapeName=dyn_tube_bck_03_capsule, center=<-0.7929 0 0>, rotation=<rot 90 0 0>}]
 dyn_tube_bck_04 [{shapeType=Capsule, length=3.5, radius=1.5, shapeName=dyn_tube_bck_04_capsule, center=<-0.7929 0 0>, rotation=<rot 90 23.3383 8.3383>}]
 dyn_tube_c [{shapeType=Capsule, length=3.5, radius=1.5, shapeName=dyn_tube_c_capsule, rotation=<rot 0 0 90>}]
 dyn_tube_fr_01 [{shapeType=Capsule, length=0.505, radius=0.505, shapeName=dyn_tube_fr_01_capsule, rotation=<rot 0 -90 0>}]
 dyn_tube_fr_02 [{shapeType=Capsule, length=3.5, radius=1.5, shapeName=dyn_tube_fr_02_capsule, center=<1.6081 0 0>, rotation=<rot 90 0 0>}]
 dyn_tube_fr_03 [{shapeType=Capsule, length=3.5, radius=1.5, shapeName=dyn_tube_fr_03_capsule, center=<1.6081 0 0>, rotation=<rot 90 0 0>}]
 dyn_tube_fr_04 [{shapeType=Capsule, length=3.5, radius=1.5, shapeName=dyn_tube_fr_04_capsule, center=<1.6081 0 0>, rotation=<rot 90 0 0>}]
 dyn_tube_fr_05 [{shapeType=Capsule, length=3.5, radius=1.5, shapeName=dyn_tube_fr_05_capsule, center=<1.6081 0 0>, rotation=<rot 90 0 0>}]
 dyn_tube_fr_06 [{shapeType=Capsule, length=3.5, radius=1.5, shapeName=dyn_tube_fr_06_capsule, center=<1.6081 0 0>, rotation=<rot 90 0 0>}]
 dyn_tube_fr_07 [{shapeType=Capsule, length=3.5, radius=1.5, shapeName=dyn_tube_fr_07_capsule, center=<1.6081 0 0>, rotation=<rot 90 0 0>}]
 lowerarm_l [{shapeType=Capsule, length=20.5, radius=4, shapeName=lowerarm_l_capsule, center=<13.6203 0.0001 0>, rotation=<rot 90 -166.7373 -166.7376>}]
 spine_01 [{shapeType=Capsule, length=5, radius=3, shapeName=spine_01_capsule}]
 spine_02 [{shapeType=Capsule, length=6, radius=10.5, shapeName=spine_02_capsule, center=<0 1.571 0>}]
 spine_03 [{shapeType=Capsule, length=3.5, radius=12.5, shapeName=spine_03_capsule, center=<0 1.6304 0>}]
 spine_04 [{shapeType=Capsule, length=4.5, radius=15, shapeName=spine_04_capsule, center=<1.6591 1.5155 0>}]
 spine_05 [{shapeType=Capsule, length=9, radius=15.5, shapeName=spine_05_capsule, center=<-5.0397 0.9758 0>}]
 upperarm_l [{shapeType=Capsule, length=20, radius=5.5, shapeName=upperarm_l_capsule, center=<14.0341 0.7872 0>, rotation=<rot 90 -179.9999 179.9998>}]
}
constraints [
 {
  bone1Name=dyn_tube_bck_01
  bone2Name=spine_03
  swing1LimitDegrees=45
  swing1Motion=Free
  swing2LimitDegrees=45
  swing2Motion=Free
  twistLimitDegrees=45
  twistMotion=Free
 }
 {
  bone1Name=dyn_tube_bck_02
  bone2Name=dyn_tube_bck_01
  swing1LimitDegrees=45
  swing1Motion=Free
  swing2LimitDegrees=45
  swing2Motion=Free
  twistLimitDegrees=1
 }
 {
  bone1Name=dyn_tube_bck_03
  bone2Name=dyn_tube_bck_02
  swing1LimitDegrees=45
  swing1Motion=Free
  swing2LimitDegrees=45
  swing2Motion=Free
  twistLimitDegrees=1
 }
 {
  bone1Name=dyn_tube_bck_04
  bone2Name=dyn_tube_bck_03
  swing1LimitDegrees=45
  swing1Motion=Free
  swing2LimitDegrees=45
  swing2Motion=Free
  twistLimitDegrees=1
 }
 {
  bone1Name=dyn_tube_bck_04
  bone2Name=dyn_tube_c
  swing1LimitDegrees=45
  swing1Motion=Free
  swing2LimitDegrees=45
  swing2Motion=Free
  twistLimitDegrees=45
  twistMotion=Free
 }
 {
  bone1Name=dyn_tube_c
  bone2Name=dyn_tube_bck_04
  swing1LimitDegrees=45
  swing1Motion=Free
  swing2LimitDegrees=45
  swing2Motion=Free
  twistLimitDegrees=45
  twistMotion=Free
 }
 {
  bone1Name=dyn_tube_fr_07
  bone2Name=dyn_tube_c
  swing1LimitDegrees=45
  swing1Motion=Free
  swing2LimitDegrees=45
  swing2Motion=Free
  twistLimitDegrees=45
  twistMotion=Free
 }
 {
  bone1Name=dyn_tube_c
  bone2Name=dyn_tube_fr_07
  swing1LimitDegrees=45
  swing1Motion=Free
  swing2LimitDegrees=45
  swing2Motion=Free
  twistLimitDegrees=45
  twistMotion=Free
 }
 {
  bone1Name=dyn_tube_fr_01
  bone2Name=spine_05
  swing1LimitDegrees=45
  swing1Motion=Free
  swing2LimitDegrees=45
  swing2Motion=Free
  twistLimitDegrees=10
 }
 {
  bone1Name=dyn_tube_fr_02
  bone2Name=dyn_tube_fr_01
  swing1LimitDegrees=45
  swing1Motion=Free
  swing2LimitDegrees=45
  swing2Motion=Free
  twistLimitDegrees=1
 }
 {
  bone1Name=dyn_tube_fr_03
  bone2Name=dyn_tube_fr_02
  swing1LimitDegrees=45
  swing1Motion=Free
  swing2LimitDegrees=45
  swing2Motion=Free
  twistLimitDegrees=1
 }
 {
  bone1Name=dyn_tube_fr_04
  bone2Name=dyn_tube_fr_03
  swing1LimitDegrees=45
  swing1Motion=Free
  swing2LimitDegrees=45
  swing2Motion=Free
  twistLimitDegrees=1
 }
 {
  bone1Name=dyn_tube_fr_05
  bone2Name=dyn_tube_fr_04
  swing1LimitDegrees=45
  swing1Motion=Free
  swing2LimitDegrees=45
  swing2Motion=Free
  twistLimitDegrees=1
 }
 {
  bone1Name=dyn_tube_fr_06
  bone2Name=dyn_tube_fr_05
  swing1LimitDegrees=45
  swing1Motion=Free
  swing2LimitDegrees=45
  swing2Motion=Free
  twistLimitDegrees=1
 }
 {
  bone1Name=dyn_tube_fr_07
  bone2Name=dyn_tube_fr_06
  swing1LimitDegrees=45
  swing1Motion=Free
  swing2LimitDegrees=45
  swing2Motion=Free
  twistLimitDegrees=1
 }
 {
  bone1Name=dyn_pouch_rotate
  bone2Name=spine_01
  swing1LimitDegrees=45
  swing1Motion=Free
  swing2LimitDegrees=45
  swing2Motion=Free
  twistLimitDegrees=45
  twistMotion=Free
 }
 {
  bone1Name=dyn_pouch_bck
  bone2Name=dyn_pouch_rotate
  swing1LimitDegrees=45
  swing1Motion=Free
  swing2LimitDegrees=45
  swing2Motion=Free
  twistLimitDegrees=45
  twistMotion=Free
 }
 {
  bone1Name=dyn_pouch_fr
  bone2Name=dyn_pouch_rotate
  swing1LimitDegrees=45
  swing1Motion=Free
  swing2LimitDegrees=45
  swing2Motion=Free
  twistLimitDegrees=45
  twistMotion=Free
 }
]
