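# Post-style AnimBlueprint for the SynthContact medium male player body, copying the attached parent's pose and adding secondary motion.
# A Control Rig, fed a parachute-not-opened value by open/close events, leads into a RigidBody sim based at spine_03.
# Sprung AnimDynamics and a two-target constraint drive the tube bones; the cached result then has dyn_blade scaled to zero.
# Right-arm events set a flag that the AnimGraph never reads.
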AnimBlueprint /BRCosmetics/Characters/Player/Male/Medium/Bodies/M_MED_SynthContact/Meshes/M_MED_SynthContact_AnimBP
target_skeleton=&/M_MED_SynthContact_Skeleton
graphs [
 AnimGraph {nodes [
   n1 = "Transform (Modify) Bone - Bone: dyn_blade"(ComponentPose=Local To Component(LocalPose=Use cached pose 'DynamicsCache'()), Alpha=1, Translation="0,0,0", Rotation="0,0,0", Scale="0,0,0") {scaleMode=BMM_Replace, scaleSpace=BCS_BoneSpace, alphaBoolBlend {blendOption=Linear}, boneToModify {boneName=dyn_blade}, scale=<0 0 0>}
   Output Pose(Result=Component To Local(ComponentPose=@n1)) {layerGroup=None}
   n2 = Control Rig!(Source=Copy Pose From Mesh() {bUseAttachedParent=true}, Alpha, Parachute Not Opened:double=Get is parachute() -> (is parachute:double)) -> (Pose) {
    alphaBoolBlend {blendOption=Linear}
    controlRigAssetReference {blueprintRigClass=&/M_MED_SynthContact_CtrlRig.M_MED_SynthContact_CtrlRig_C}
   }
   n3 = RigidBody(ComponentPose=Local To Component(LocalPose=@n2)) {
    bClampLinearTranslationLimitToRefPose=true
    bOverrideWorldGravity=true
    overridePhysicsAsset=&/M_MED_SynthContact_Physics
    simulationSpace=BaseBoneSpace
    alphaBoolBlend {blendOption=Linear}
    baseBoneRef {boneName=spine_03}
    overrideWorldGravity {z=-980}
    simSpaceSettings {maxAngularAcceleration=1000, maxAngularVelocity=8, maxLinearAcceleration=1000, maxLinearVelocity=800, worldAlpha=0.3}
   }
   n4 = Anim Dynamics - Bone: dyn_tube_c(ComponentPose=@n3, Alpha=1, ExternalForce="0,0,0") {
    angularDampingOverride=0.99
    angularSpringConstant=2000
    bAngularSpring=true
    bOverrideAngularDamping=true
    simulationSpace=RootRelative
    alphaBoolBlend {blendOption=Linear}
    boundBone {boneName=dyn_tube_c}
    componentAppliedLinearAccClamp=<3000 3000 3000>
    componentLinearAccScale=<1 1 1>
    componentLinearVelScale=<1 1 1>
    physicsBodyDefinitions [{
      boundBone {boneName=dyn_tube_c}
      boxExtents=<10 10 10>
      constraintSetup {angularLimitsMax=<0 10 0>, angularLimitsMin=<0 -10 0>, angularTarget=<0 0 0>, linearAxesMax=<1 0 0>, linearAxesMin=<-1 0 0>}
      localJointOffset=<0.1 0 0>
     }]
   }
   n5 = Anim Dynamics - Bone: dyn_tube_bck_04(ComponentPose=@n4, Alpha=1, ExternalForce="0,0,0") {
    angularDampingOverride=0.99
    angularSpringConstant=2000
    bAngularSpring=true
    bOverrideAngularDamping=true
    simulationSpace=RootRelative
    alphaBoolBlend {blendOption=Linear}
    boundBone {boneName=dyn_tube_bck_04}
    componentAppliedLinearAccClamp=<3000 3000 3000>
    componentLinearAccScale=<1 1 1>
    componentLinearVelScale=<1 1 1>
    physicsBodyDefinitions [{
      boundBone {boneName=dyn_tube_bck_04}
      boxExtents=<10 10 10>
      constraintSetup {angularLimitsMax=<0 10 0>, angularLimitsMin=<0 -10 0>, angularTarget=<0 0 0>}
      localJointOffset=<0.1 0 0>
     }]
   }
   n6 = Anim Dynamics - Bone: dyn_tube_bck_03(ComponentPose=@n5, Alpha=1, ExternalForce="0,0,0") {
    angularDampingOverride=0.99
    angularSpringConstant=2000
    bAngularSpring=true
    bOverrideAngularDamping=true
    simulationSpace=RootRelative
    alphaBoolBlend {blendOption=Linear}
    boundBone {boneName=dyn_tube_bck_03}
    componentAppliedLinearAccClamp=<3000 3000 3000>
    componentLinearAccScale=<1 1 1>
    componentLinearVelScale=<1 1 1>
    physicsBodyDefinitions [{
      boundBone {boneName=dyn_tube_bck_03}
      boxExtents=<10 10 10>
      constraintSetup {angularLimitsMax=<0 10 0>, angularLimitsMin=<0 -10 0>, angularTarget=<0 0 0>}
      localJointOffset=<0.1 0 0>
     }]
   }
   n7 = Anim Dynamics - Bone: dyn_tube_fr_07(ComponentPose=@n6, Alpha=1, ExternalForce="0,0,0") {
    angularDampingOverride=0.99
    angularSpringConstant=2000
    bAngularSpring=true
    bOverrideAngularDamping=true
    simulationSpace=RootRelative
    alphaBoolBlend {blendOption=Linear}
    boundBone {boneName=dyn_tube_fr_07}
    componentAppliedLinearAccClamp=<3000 3000 3000>
    componentLinearAccScale=<1 1 1>
    componentLinearVelScale=<1 1 1>
    physicsBodyDefinitions [{
      boundBone {boneName=dyn_tube_fr_07}
      boxExtents=<10 10 10>
      constraintSetup {angularLimitsMax=<0 10 0>, angularLimitsMin=<0 -10 0>, angularTarget=<0 0 0>}
      localJointOffset=<0.1 0 0>
     }]
   }
   n8 = "Constraint - dyn_tube_c (PP)"!(ComponentPose=@n7, Alpha=1, ConstraintWeights_0=0.5, ConstraintWeights_1:float=0.5) -> (Pose) {
    alphaBoolBlend {blendOption=Linear}
    boneToModify {boneName=dyn_tube_c}
    constraintSetup [{targetBone {boneName=dyn_tube_bck_04}}, {targetBone {boneName=dyn_tube_fr_07}}]
    constraintWeights [1,1]
   }
   Save cached pose 'DynamicsCache'(Pose=Component To Local(ComponentPose=@n8)) {cacheName=DynamicsCache}
  ]}
 EventGraph {nodes [
   Event OnParachuteOpened()
   > Set is parachute(is parachute:double=0) -> (Output_Get:double)
   Event OnParachuteClosed()
   > Set is parachute(is parachute=1)
   Event OnRightArmHidden() {is=title K2Node_Event}
   > Set Is Right Arm Hidden(Is Right Arm Hidden:bool=true) -> (Output_Get:bool)
   Event OnRightArmVisible() {is=title K2Node_Event}
   > Set Is Right Arm Hidden(Is Right Arm Hidden=false)
   Event BlueprintUpdateAnimation()
  ]}
]
variables [
 is parachute {type=number, default=1}
 Is Right Arm Hidden {type=boolean}
]
