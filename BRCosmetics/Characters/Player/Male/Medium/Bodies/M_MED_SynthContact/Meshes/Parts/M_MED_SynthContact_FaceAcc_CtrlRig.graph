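# Control Rig for the SynthContact male medium face accessory that adjusts the right head tube as the head moves.
# A three-step sequence runs a spherical pose reader on the head bone, stores its output in HeadUp, then uses that as the weight for a rotation on dyn_headTube_base_r.
# The AlphaInterp between them has interpolation and clamping off.
# The left tube is not driven, and the reader's debug draw is enabled.
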ControlRigBlueprint /BRCosmetics/Characters/Player/Male/Medium/Bodies/M_MED_SynthContact/Meshes/Parts/M_MED_SynthContact_FaceAcc_CtrlRig
graphs [
 RigVMFunctionLibrary {links [], nodes []}
 RigVMFunction_Sequence_1_ContainedGraph {nodes [
   n1 = RigVMFunctionEntryNode()
   RigVMFunction_Sequence = &RigVMFunction_Sequence(ExecuteContext=@n1.ExecuteContext)
   n2 = &RigVMFunction_Sequence(ExecuteContext=@RigVMFunction_Sequence.B)
   RigVMFunctionReturnNode(A=@RigVMFunction_Sequence.A, B=@n2.A, C=@n2.B)
  ]}
 RigVMModel {nodes [
   RigUnit_BeginExecution = &RigUnit_BeginExecution()
   RigVMFunction_Sequence = RigVMAggregateNode(ExecuteContext=@RigUnit_BeginExecution.ExecutePin)
   n1 = &RigUnit_SphericalPoseReader(
    ActiveRegionScaleFactors=(PositiveWidth=1,NegativeWidth=1,PositiveHeight=1,NegativeHeight=1)
    ActiveRegionSize=0.2
    Debug=(bDrawDebug=true,bDraw2D=False,bDrawLocalAxes=False,DebugScale=25,DebugSegments=20,DebugThickness=0.25)
    DriverAxis=(X=1,Y=0,Z=0)
    DriverItem=(Type=Bone,Name="head")
    FalloffRegionScaleFactors=(PositiveWidth=1.5,NegativeWidth=1,PositiveHeight=1,NegativeHeight=1)
    FalloffSize=0.32
    FlipHeightScaling=False
    FlipWidthScaling=False
    OptionalParentItem=(Type=Bone,Name="None")
    RotationOffset=(X=-64.889,Y=15.1763,Z=36.1456)
    ExecutePin=@RigVMFunction_Sequence.A
   )
   RigVMVariableNode(ExecuteContext=@RigVMFunction_Sequence.B, Value=@n1.OutputParam) {variable=HeadUp}
   AlphaInterp = &RigVMFunction_AlphaInterp(
    InterpSpeedDecreasing=10
    InterpSpeedIncreasing=10
    Scale=1
    bClampResult=False
    bInterpResult=False
    Value=RigVMVariableNode() {variable=HeadUp}.Value
   )
   &RigUnit_ModifyTransforms(
    ItemToModify=((Item=(Type=Bone,Name="dyn_headTube_base_r"),Transform=(Rotation=(X=0.1722,Y=-0.1285,Z=0.0227,W=0.9764),Translation=(X=0,Y=0,Z=0),Scale3D=(X=1,Y=1,Z=1))))
    WeightMinimum=0
    Weight=@AlphaInterp.Result
    ExecutePin=@RigVMFunction_Sequence.C
   )
  ]}
]
hierarchy [
 root {loc=<0 0 0>, rot=<0 0 0 1>}
 pelvis {parent=root, loc=<0 -0.7071 93.1191>, rot=<0.0127 0.707 -0.0127 -0.707>}
 spine_01 {parent=pelvis, loc=<6.7494 -0.6388 0>, rot=<0 0 0.034 -0.9994>}
 spine_02 {parent=spine_01, loc=<7.2518 -0.035 0>, rot=<0 0 0.0238 -0.9997>}
 spine_03 {parent=spine_02, loc=<6.0104 0 0>, rot=<0 0 0.0149 -0.9999>}
 spine_04 {parent=spine_03, loc=<8.1024 0 0>, rot=<0 0 -0.0132 -0.9999>}
 spine_05 {parent=spine_04, loc=<10.0146 0 0>, rot=<0 0 -0.0015 -1>}
 neck_01 {parent=spine_05, loc=<11.3138 0.1529 0>, rot=<0 0 -0.2003 -0.9797>}
 neck_02 {parent=neck_01, loc=<4.5627 0 0>, rot=<0 0 0.0504 -0.9987>}
 head {parent=neck_02, loc=<5.3914 0 0>, rot=<0 0 0.115 -0.9934>}
 dyn_headTube_base_l {parent=head, loc=<-0.0104 -2.7952 -5.88>, rot=<0.1853 -0.4797 0.1963 -0.8349>}
 dyn_headTube_l_01 {parent=dyn_headTube_base_l, loc=<1.7858 2.2649 -0.65>, rot=<0.1862 -0.19 -0.2699 -0.9254>}
 dyn_headTube_l_02 {parent=dyn_headTube_base_l, loc=<5.1989 -3.003 0.6567>, rot=<0.1028 -0.1163 0.0634 -0.9858>}
 dyn_headTube_base_r {parent=head, loc=<-0.0104 -2.7952 5.88>, rot=<0.823 -0.1755 0.5203 -0.1455>}
 dyn_headTube_r_01 {parent=dyn_headTube_base_r, loc=<2.0194 -2.0907 -0.5411>, rot=<-0.2339 -0.1234 0.301 -0.9162>}
 dyn_headTube_r_02 {parent=dyn_headTube_base_r, loc=<4.8518 3.3978 1.1797>, rot=<-0.0812 -0.1005 0.0107 -0.9916>}
 clavicle_l {parent=spine_05, loc=<4.3575 -0.1466 -2.3223>, rot=<0.0465 -0.7289 -0.009 -0.6829>}
 clavicle_r {parent=spine_05, loc=<4.3575 -0.1466 2.3223>, rot=<0.7289 0.0465 -0.6829 0.009>}
]
variables [HeadUp {type=double, default=0}]
